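# Niagara effect for the YuzuCrank glider deploy, drawing an orange glowing frame and glows around the glider's Body_Offset socket.
# Frame spawns meshes in staggered bursts, with per-group scale curves and platform-dependent color intensity.
# FrameGlow adds two offset sprites, Glows moves six partitioned sprites along per-index curves, and RedGlows adds four delayed, aligned red sprites.
# Front-end modifiers alter glow placement and alpha, while time-of-day scalars adjust brightness.
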NiagaraSystem /BRCosmetics/Gadgets/Assets/VinderTech_GliderChute/Glider_YuzuCrank/FX/NS_Glider_YuzuCrank_Deploy
dependencies {
 usedDataInterfaces [
  &NiagaraDataInterfaceCurve
  &NiagaraDataInterfaceSocketReader
  &NiagaraDataInterfaceColorCurve
  &NiagaraDataInterfacePlatformSet
  &NiagaraDataInterfaceMeshRendererInfo
  &NiagaraDataInterfaceVectorCurve
 ]
 usedDynamicInputs []
 usedModules [
  &self:SystemSpawnScript.SystemScriptSource.SystemScriptGraph.NiagaraNodeAssignment_0.SetVariables_#1
  &SystemState
 ]
}
emitters {
 Frame {
  emitterSpawnScript [/SetVariables(
    Emitter.Color:LinearColor=&LerpLinearColors(
     StartColor=<rgba 1 0.2224 0 1>
     EndColor=<rgba 8 1.7794 0 1>
     LerpFactor=&ClampFloat(Float=&Add_Float(
       A=&FN_PickFloatDependOnPlatform(
        TargetPlatform=DataInterface() {propertyValues {Platforms {
           qualityLevelMask=-1
           deviceProfileStates [
            {profileName=Windows, qualityLevelMask=0, setQualityLevelMask=3}
            {profileName=Mac, qualityLevelMask=0, setQualityLevelMask=3}
            {profileName=Linux, qualityLevelMask=0, setQualityLevelMask=3}
           ]
          }}}
        TargetPlatformActive=1
        TargetPlatformNoActive=0
       )
       B=&FN_PickFloatDependOnPlatform(
        TargetPlatform=DataInterface() {propertyValues {Platforms {
           qualityLevelMask=-1
           cVarConditions [{
             bUseMaxFloat=false
             bUseMaxInt=false
             bUseMinFloat=true
             bUseMinInt=true
             cVarName=athena.isperfmode
             failResponse=Disable
             maxFloat=1
             maxInt=1
             minFloat=1
             minInt=1
             passResponse=None
             value=true
            }]
          }}}
        TargetPlatformActive=1
        TargetPlatformNoActive=0
       )
      ))
    )
    Emitter.HDRMult:NiagaraFloat=&FNTimeOfDayScalar(Night Scale=1, Morning Scale=0.5, Day Scale=0.05, Evening Scale=0.5)
   ) {isSetParameters=true}]
  emitterUpdateScript [
   &EmitterState(
    Inactive Response=::NewEnumerator1
    Life Cycle Mode=::NewEnumerator1
    Loop Behavior=::NewEnumerator1
    Loop Delay=0.3
    Loop Duration=2
    UseLoopDelay=false
   )
   &SpawnBurst_Instantaneous(Spawn Group=3, Spawn Time=0.4, Use Spawn Probability=false, UseLoopCountLimit=false) {module=SpawnBurst_Instantaneous003}
   &SpawnBurst_Instantaneous(Spawn Group=2, Spawn Time=0.4, Use Spawn Probability=false, UseLoopCountLimit=false) {module=SpawnBurst_Instantaneous002}
   &SpawnBurst_Instantaneous(Spawn Group=1, Spawn Time=0.2, Use Spawn Probability=false, UseLoopCountLimit=false) {module=SpawnBurst_Instantaneous001}
   &SpawnBurst_Instantaneous(Use Spawn Probability=false, UseLoopCountLimit=false)
  ]
  particleSpawnScript [&InitializeParticle(
    Lifetime=&Subtract_Float(
     A=1.65
     B=&MergeHelper_FloatSwitch(0=0, 1=0.2, 2=0.4, 3=0.4, Index=&SpawnGroup(), Num=::NewEnumerator2)
    )
    Mesh Renderer Array Visibility Mode=::NewEnumerator1
    Mesh Renderer Info=DataInterface() {propertyValues {MeshRenderer=&self:Frame_0.NiagaraMeshRendererProperties_0}}
    Mesh Renderer Mesh Index:NiagaraInt32=&SpawnGroup()
    Mesh Scale Mode=::NewEnumerator3
    UsePositionOffset=false
    Write Material Random=true
   )]
  particleUpdateScript [
   &ParticleState()
   &ScaleMeshSize(Scale Factor=&MergeHelper_VectorSwitch(
     0=&VectorFromCurve(VectorCurve=DataInterface() {propertyValues {
        XCurve {keys [<at 0 -1 RCIM_Cubic>, {arriveTangent=-0.1954, interpMode=RCIM_Cubic, leaveTangent=-0.1954, tangentMode=RCTM_User, time=0.3, value=1}]}
        YCurve {keys [<at 0 -1 RCIM_Cubic>,<at 0.3 1 RCIM_Cubic>]}
        ZCurve {keys [<at 0 1>,<at 1 1>]}
       }})
     1=&VectorFromCurve(VectorCurve=DataInterface() {propertyValues {
        XCurve {keys [<at 0 -0.5 RCIM_Cubic>, {arriveTangent=-0.1954, interpMode=RCIM_Cubic, leaveTangent=-0.1954, tangentMode=RCTM_User, time=0.4, value=0.9918}]}
        YCurve {keys [<at 0 -0.5 RCIM_Cubic>,<at 0.4 1 RCIM_Cubic>]}
        ZCurve {keys [<at 0 1>,<at 1 1>]}
       }})
     2=&VectorFromCurve(VectorCurve=DataInterface() {propertyValues {
        XCurve {keys [<at 0 0 RCIM_Cubic>, {arriveTangent=-0.1954, interpMode=RCIM_Cubic, leaveTangent=-0.1954, tangentMode=RCTM_User, time=0.4, value=0.9918}]}
        YCurve {keys [<at 0 0 RCIM_Cubic>,<at 0.4 1 RCIM_Cubic>]}
        ZCurve {keys [<at 0 1>,<at 1 1>]}
       }})
     3=&VectorFromCurve(VectorCurve=DataInterface() {propertyValues {
        XCurve {keys [<at 0 -1 RCIM_Cubic>, {arriveTangent=-0.1954, interpMode=RCIM_Cubic, leaveTangent=-0.1954, tangentMode=RCTM_User, time=0.4, value=0.9918}]}
        YCurve {keys [<at 0 -1 RCIM_Cubic>,<at 0.4 1 RCIM_Cubic>]}
        ZCurve {keys [<at 0 1>,<at 1 1>]}
       }})
     Index=&SpawnGroup()
     Num=::NewEnumerator2
    ))
   &SocketLocation(
    SamplePosOffset=true
    Sampled Position Offset=<0 -15 -7>
    Socket Reader=DataInterface() {propertyValues {
      AttachComponentClass=null
      AttachComponentTag=None
      EditorPreviewAsset=/BRCosmetics/Gadgets/Assets/VinderTech_GliderChute/Glider_YuzuCrank/Meshes/YuzuCrank_Glider
      SourceActorRef=null
      SourceAsset=null
      SourceMode=Default
      bRequireCurrentFrameData=true
      bUpdateSocketsPerFrame=true
      FilteredSockets [Body_Offset]
      ObjectParameterBinding {parameter {name=None, typeDefHandle {registeredTypeIndex=1}, varData []}}
     }}
    Socket Sampling Mode=::NewEnumerator3
    VelocitySpeedLimit=false
   )
   &DynamicMaterialParameters(
    Index 0 Param 1=&Multiply_Float(
     A=&FloatFromCurve(FloatCurve=DataInterface() {propertyValues {Curve {
         defaultValue=1
         keys [
          <at 0 0 RCIM_Cubic>
          {arriveTangent=1.5345, interpMode=RCIM_Cubic, leaveTangent=1.5345, time=0.396, value=0.2439}
          {arriveTangent=-0.0405, interpMode=RCIM_Cubic, leaveTangent=-0.0405, tangentMode=RCTM_User, time=0.6517, value=1}
          {arriveTangent=-3.1619, interpMode=RCIM_Cubic, leaveTangent=-3.1619, tangentMode=RCTM_User, time=1, value=0}
         ]
        }}})
     B=&Waveform("Waveform [1]"=::NewEnumerator0, "[1] Frequency"=0, "[1] Amplitude Min/Max"=<0.4 1>)
    )
    Index 0 Param 2=2
    Index 0 Param 3=1
    Index 0 Param 4=5
    Module.Index0Float0Write=true
    Module.Index0Float1Write=true
    Module.Index0Float2Write=true
    Module.Index0Float3Write=true
   )
   &ScaleColor(
    Scale Alpha=&FloatFromCurve(
     FloatCurve=DataInterface() {propertyValues {Curve {
        defaultValue=1
        keys [
         <at 0 0 RCIM_Cubic>
         {arriveTangent=2.1774, interpMode=RCIM_Cubic, leaveTangent=2.1774, time=0.2039, value=0.3312}
         {arriveTangentWeight=0.2667, interpMode=RCIM_Cubic, leaveTangentWeight=0.1083, tangentMode=RCTM_User, time=0.4593, value=1}
         {arriveTangent=-0.0405, interpMode=RCIM_Cubic, leaveTangent=-0.0405, tangentMode=RCTM_User, time=0.5152, value=1}
         {arriveTangent=-3.1619, interpMode=RCIM_Cubic, leaveTangent=-3.1619, tangentMode=RCTM_User, time=1, value=0}
        ]
       }}}
     Scale Curve=2
    )
    ScaleA=true
    ScaleRGB=false
   )
  ]
  rendererClasses [&NiagaraMeshRendererProperties]
  renderers [{rendererClass=&NiagaraMeshRendererProperties, rendererIndex=0}]
 }
 FrameGlow {
  emitterUpdateScript [
   &EmitterState(
    Inactive Response=::NewEnumerator1
    Life Cycle Mode=::NewEnumerator1
    Loop Behavior=::NewEnumerator1
    Loop Delay=0.1
    Loop Duration=2
    UseLoopDelay=true
   )
   &SpawnBurst_Instantaneous(Spawn Count=2, Use Spawn Probability=false, UseLoopCountLimit=false)
  ]
  particleSpawnScript [
   &PartitionParticles()
   &InitializeParticle(Color Mode=::NewEnumerator0, Sprite Size Mode=::NewEnumerator3, UsePositionOffset=false, Write Material Random=true)
   &DynamicMaterialParameters(
    Index 0 Param 1=2
    Index 0 Param 2=10
    Module.Index0Float0Write=true
    Module.Index0Float1Write=true
    Module.Index0Float2Write=true
    Module.Index0Float3Write=true
   )
  ]
  particleUpdateScript [
   &ParticleState()
   &ScaleSpriteSize(Uniform Curve Sprite Scale=DataInterface() {propertyValues {Curve {
       defaultValue=0
       keys [{interpMode=RCIM_Cubic, tangentMode=RCTM_User, time=0, value=0}, {interpMode=RCIM_Cubic, tangentMode=RCTM_User, time=0.1, value=1}]
      }}})
   &SocketLocation(
    SamplePosOffset=true
    Sampled Position Offset=&MaskVector(Mask Source=1, Mask Parameter=$Particles.PartitionParticles.Partition, Mask=0, Masked Vector=<-55 25 0>, Passthrough Vector=<55 25 0>)
    Socket Reader=DataInterface() {propertyValues {
      AttachComponentClass=null
      AttachComponentTag=None
      EditorPreviewAsset=&YuzuCrank_Glider
      SourceActorRef=null
      SourceAsset=null
      SourceMode=Default
      bRequireCurrentFrameData=true
      bUpdateSocketsPerFrame=true
      FilteredSockets [Body_Offset]
      ObjectParameterBinding {parameter {name=None, typeDefHandle {registeredTypeIndex=1}, varData []}}
     }}
    Socket Sampling Mode=::NewEnumerator3
    VelocitySpeedLimit=false
   )
   &ScaleColor(
    Scale Alpha=&FloatFromCurve(FloatCurve=DataInterface() {propertyValues {Curve {
        defaultValue=1
        keys [
         {arriveTangentWeight=0.1965, interpMode=RCIM_Cubic, leaveTangentWeight=0.1761, tangentMode=RCTM_User, time=0.2176, value=1}
         {interpMode=RCIM_Cubic, tangentMode=RCTM_User, time=0.5, value=1}
         {tangentMode=RCTM_User, time=1, value=0}
        ]
       }}})
    Scale RGB=&MakeVectorFromLinearColorRGB(Color=<rgba 1.5 0.3026 0 1>)
    ScaleA=true
    ScaleRGB=true
   )
  ]
  rendererClasses [&NiagaraSpriteRendererProperties]
  renderers [{rendererClass=&NiagaraSpriteRendererProperties}]
 }
 Glows {
  emitterUpdateScript [
   &EmitterState(
    Inactive Response=::NewEnumerator1
    Life Cycle Mode=::NewEnumerator1
    Loop Behavior=::NewEnumerator1
    Loop Delay=$System.Delay
    Loop Duration=2
    UseLoopDelay=true
   )
   &SpawnBurst_Instantaneous(Spawn Count=6, Use Spawn Probability=false, UseLoopCountLimit=false)
  ]
  particleSpawnScript [
   &PartitionParticles(Number Of Partitions=6, Particles Per Partition=1, Partition Mode:ENiagara_PartitionMode_v2=::NewEnumerator1)
   &InitializeParticle(
    Color Mode=::NewEnumerator0
    Lifetime=&MergeHelper_FloatSwitch(0=0.8, 1=0.8, 2=0.3, 3=0.3, 4=0.8, 5=0.8, Index=$Particles.PartitionParticles.Partition, Num=::NewEnumerator4)
    Sprite Size Mode=::NewEnumerator3
    Uniform Sprite Size=100
    UsePositionOffset=false
    Write Material Random=true
   )
   &DynamicMaterialParameters(
    Index 0 Param 1=2
    Index 0 Param 2=10
    Module.Index0Float0Write=true
    Module.Index0Float1Write=true
    Module.Index0Float2Write=true
    Module.Index0Float3Write=true
   )
   &CameraOffset(Camera Offset Amount=20)
  ]
  particleUpdateScript [
   &ParticleState()
   &ScaleSpriteSize(
    Scale Sprite Size Mode=::NewEnumerator0
    Uniform Scale Factor=&MergeHelper_FloatSwitch(
     0=&FloatFromCurve(FloatCurve=DataInterface() {propertyValues {Curve {
         defaultValue=0
         keys [
          {arriveTangent=-3.8656, interpMode=RCIM_Cubic, leaveTangent=-3.8656, tangentMode=RCTM_User, time=0, value=2}
          {arriveTangent=-0.7284, interpMode=RCIM_Cubic, leaveTangent=-0.7284, tangentMode=RCTM_User, time=1, value=0.8}
         ]
        }}})
     1=&FloatFromCurve(FloatCurve=DataInterface() {propertyValues {Curve {
         defaultValue=0
         keys [
          {arriveTangent=-3.8656, interpMode=RCIM_Cubic, leaveTangent=-3.8656, tangentMode=RCTM_User, time=0, value=2}
          {arriveTangent=-0.7284, interpMode=RCIM_Cubic, leaveTangent=-0.7284, tangentMode=RCTM_User, time=1, value=0.8}
         ]
        }}})
     2=&FloatFromCurve(FloatCurve=DataInterface() {propertyValues {Curve {
         defaultValue=0
         keys [
          {arriveTangent=-1.556, interpMode=RCIM_Cubic, leaveTangent=-1.556, tangentMode=RCTM_User, time=0, value=1.5}
          {arriveTangent=-0.7284, interpMode=RCIM_Cubic, leaveTangent=-0.7284, tangentMode=RCTM_User, time=1, value=0.8}
         ]
        }}})
     3=&FloatFromCurve(FloatCurve=DataInterface() {propertyValues {Curve {
         defaultValue=0
         keys [
          {arriveTangent=-1.556, interpMode=RCIM_Cubic, leaveTangent=-1.556, tangentMode=RCTM_User, time=0, value=1.5}
          {arriveTangent=-0.7284, interpMode=RCIM_Cubic, leaveTangent=-0.7284, tangentMode=RCTM_User, time=1, value=0.8}
         ]
        }}})
     4=&FloatFromCurve(FloatCurve=DataInterface() {propertyValues {Curve {
         defaultValue=0
         keys [
          {arriveTangent=-3.8656, interpMode=RCIM_Cubic, leaveTangent=-3.8656, tangentMode=RCTM_User, time=0, value=2}
          {arriveTangent=-0.7284, interpMode=RCIM_Cubic, leaveTangent=-0.7284, tangentMode=RCTM_User, time=1, value=0.6}
         ]
        }}})
     5=&FloatFromCurve(FloatCurve=DataInterface() {propertyValues {Curve {
         defaultValue=0
         keys [
          {arriveTangent=-3.8656, interpMode=RCIM_Cubic, leaveTangent=-3.8656, tangentMode=RCTM_User, time=0, value=2}
          {arriveTangent=-0.7284, interpMode=RCIM_Cubic, leaveTangent=-0.7284, tangentMode=RCTM_User, time=1, value=0.6}
         ]
        }}})
     Index=$Particles.PartitionParticles.Partition
     Num=::NewEnumerator4
    )
   )
   &SocketLocation(
    SamplePosOffset=true
    Sampled Position Offset=<0 -15 -7>
    Socket Reader=DataInterface() {propertyValues {
      AttachComponentClass=null
      AttachComponentTag=None
      EditorPreviewAsset=&YuzuCrank_Glider
      SourceActorRef=null
      SourceAsset=null
      SourceMode=Default
      bRequireCurrentFrameData=true
      bUpdateSocketsPerFrame=true
      FilteredSockets [Body_Offset]
      ObjectParameterBinding {parameter {name=None, typeDefHandle {registeredTypeIndex=1}, varData []}}
     }}
    Socket Sampling Mode=::NewEnumerator3
    VelocitySpeedLimit=false
   )
   &SystemLocation(
    MaskSpawnGroup=false
    Offset=&AddVectorToPosition(
     Position=$Output.SocketLocation.SampledPosition
     Vector=&MakeVector(
      X=&Multiply_Float(
       A=&MergeHelper_FloatSwitch(
        0=&FloatFromCurve(
         FloatCurve=DataInterface() {propertyValues {Curve {
            defaultValue=0
            keys [
             {arriveTangent=0.4251, interpMode=RCIM_Cubic, leaveTangent=0.4251, tangentMode=RCTM_User, time=0, value=0}
             {interpMode=RCIM_Cubic, tangentMode=RCTM_User, time=0.8552, value=1}
            ]
           }}}
         Scale Curve=-200
        )
        1=&FloatFromCurve(
         FloatCurve=DataInterface() {propertyValues {Curve {
            defaultValue=0
            keys [
             {arriveTangent=0.4251, interpMode=RCIM_Cubic, leaveTangent=0.4251, tangentMode=RCTM_User, time=0, value=0}
             {interpMode=RCIM_Cubic, tangentMode=RCTM_User, time=0.8552, value=1}
            ]
           }}}
         Scale Curve=200
        )
        2=&FloatFromCurve(
         FloatCurve=DataInterface() {propertyValues {Curve {
            defaultValue=0
            keys [
             {arriveTangent=0.4251, interpMode=RCIM_Cubic, leaveTangent=0.4251, tangentMode=RCTM_User, time=0, value=0.5}
             {arriveTangent=0.8931, interpMode=RCIM_Cubic, leaveTangent=0.8931, tangentMode=RCTM_User, time=0.8552, value=1}
            ]
           }}}
         Scale Curve=-65
        )
        3=&FloatFromCurve(
         FloatCurve=DataInterface() {propertyValues {Curve {
            defaultValue=0
            keys [
             {arriveTangent=0.4251, interpMode=RCIM_Cubic, leaveTangent=0.4251, tangentMode=RCTM_User, time=0, value=0.5}
             {arriveTangent=0.667, interpMode=RCIM_Cubic, leaveTangent=0.667, tangentMode=RCTM_User, time=0.8552, value=1}
            ]
           }}}
         Scale Curve=65
        )
        4=&FloatFromCurve(
         FloatCurve=DataInterface() {propertyValues {Curve {
            defaultValue=0
            keys [
             {arriveTangent=0.4251, interpMode=RCIM_Cubic, leaveTangent=0.4251, tangentMode=RCTM_User, time=0, value=0}
             {interpMode=RCIM_Cubic, tangentMode=RCTM_User, time=0.8552, value=1}
            ]
           }}}
         Scale Curve=-200
        )
        5=&FloatFromCurve(
         FloatCurve=DataInterface() {propertyValues {Curve {
            defaultValue=0
            keys [
             {arriveTangent=0.4251, interpMode=RCIM_Cubic, leaveTangent=0.4251, tangentMode=RCTM_User, time=0, value=0}
             {interpMode=RCIM_Cubic, tangentMode=RCTM_User, time=0.8552, value=1}
            ]
           }}}
         Scale Curve=200
        )
        Index=$Particles.PartitionParticles.Partition
        Num=::NewEnumerator4
       )
       B=&FN_FrontEndFloatModifier(Gameplay=1, Front End=1)
      )
      Y=&FN_FrontEndFloatModifier(
       Gameplay=&Add_Float(
        A=&MergeHelper_FloatSwitch(
         0=&FloatFromCurve(
          FloatCurve=DataInterface() {propertyValues {Curve {
             defaultValue=0
             keys [{interpMode=RCIM_Cubic, tangentMode=RCTM_User, time=0, value=0}, {interpMode=RCIM_Cubic, tangentMode=RCTM_User, time=1, value=1}]
            }}}
          Scale Curve=80
         )
         1=&FloatFromCurve(
          FloatCurve=DataInterface() {propertyValues {Curve {
             defaultValue=0
             keys [{interpMode=RCIM_Cubic, tangentMode=RCTM_User, time=0, value=0}, {interpMode=RCIM_Cubic, tangentMode=RCTM_User, time=1, value=1}]
            }}}
          Scale Curve=80
         )
         2=&FloatFromCurve(
          FloatCurve=DataInterface() {propertyValues {Curve {
             defaultValue=1
             keys [{interpMode=RCIM_Cubic, tangentMode=RCTM_User, time=0, value=0.5}, {interpMode=RCIM_Cubic, tangentMode=RCTM_User, time=1, value=1}]
            }}}
          Scale Curve=150
         )
         3=&FloatFromCurve(
          FloatCurve=DataInterface() {propertyValues {Curve {
             defaultValue=1
             keys [{interpMode=RCIM_Cubic, tangentMode=RCTM_User, time=0, value=0.5}, {interpMode=RCIM_Cubic, tangentMode=RCTM_User, time=1, value=1}]
            }}}
          Scale Curve=150
         )
         Index=$Particles.PartitionParticles.Partition
         Num=::NewEnumerator2
        )
        B=10
       )
       Front End=&MergeHelper_FloatSwitch(
        0=&FloatFromCurve(
         FloatCurve=DataInterface() {propertyValues {Curve {
            defaultValue=0
            keys [{interpMode=RCIM_Cubic, tangentMode=RCTM_User, time=0, value=1}, {interpMode=RCIM_Cubic, tangentMode=RCTM_User, time=1, value=-3}]
           }}}
         Scale Curve=30
        )
        1=&FloatFromCurve(
         FloatCurve=DataInterface() {propertyValues {Curve {
            defaultValue=0
            keys [{interpMode=RCIM_Cubic, tangentMode=RCTM_User, time=0, value=1}, {interpMode=RCIM_Cubic, tangentMode=RCTM_User, time=1, value=-3}]
           }}}
         Scale Curve=30
        )
        2=&FloatFromCurve(
         FloatCurve=DataInterface() {propertyValues {Curve {
            defaultValue=1
            keys [{interpMode=RCIM_Cubic, tangentMode=RCTM_User, time=0, value=1}, {interpMode=RCIM_Cubic, tangentMode=RCTM_User, time=1, value=-2}]
           }}}
         Scale Curve=60
        )
        3=&FloatFromCurve(
         FloatCurve=DataInterface() {propertyValues {Curve {
            defaultValue=1
            keys [{interpMode=RCIM_Cubic, tangentMode=RCTM_User, time=0, value=-2}, {interpMode=RCIM_Cubic, tangentMode=RCTM_User, time=1, value=1}]
           }}}
         Scale Curve=60
        )
        Index=$Particles.PartitionParticles.Partition
        Num=::NewEnumerator2
       )
      )
      Z=&MergeHelper_FloatSwitch(
       0=&FloatFromCurve(FloatCurve=DataInterface() {propertyValues {Curve {keys [<at 0 1>,<at 1 0.5>]}}}, Scale Curve=50)
       1=&FloatFromCurve(FloatCurve=DataInterface() {propertyValues {Curve {keys [<at 0 1>,<at 1 0.5>]}}}, Scale Curve=50)
       2=&FloatFromCurve(
        FloatCurve=DataInterface() {propertyValues {Curve {keys [<at 0 1 RCIM_Cubic>, {arriveTangent=1.9193, interpMode=RCIM_Cubic, leaveTangent=1.9193, tangentMode=RCTM_User, time=1, value=3}]}}}
        Scale Curve=-10
       )
       3=&FloatFromCurve(
        FloatCurve=DataInterface() {propertyValues {Curve {keys [<at 0 1 RCIM_Cubic>, {arriveTangent=1.9193, interpMode=RCIM_Cubic, leaveTangent=1.9193, tangentMode=RCTM_User, time=1, value=3}]}}}
        Scale Curve=-10
       )
       4=&FloatFromCurve(
        FloatCurve=DataInterface() {propertyValues {Curve {defaultValue=1, keys [{tangentMode=RCTM_User, time=0, value=1}, {tangentMode=RCTM_User, time=1, value=-1}]}}}
        Scale Curve=100
       )
       5=&FloatFromCurve(
        FloatCurve=DataInterface() {propertyValues {Curve {defaultValue=1, keys [{tangentMode=RCTM_User, time=0, value=1}, {tangentMode=RCTM_User, time=1, value=-1}]}}}
        Scale Curve=100
       )
       Index=$Particles.PartitionParticles.Partition
       Num=::NewEnumerator4
      )
     )
    )
   )
   &ScaleColor(
    Scale Alpha=&Multiply_Float(
     A=&MergeHelper_FloatSwitch(
      0=&FloatFromCurve(
       FloatCurve=DataInterface() {propertyValues {Curve {
          defaultValue=1
          keys [
           {arriveTangent=5.7045, interpMode=RCIM_Cubic, leaveTangent=5.7045, tangentMode=RCTM_User, time=0.0358, value=0}
           {arriveTangentWeight=0.1895, interpMode=RCIM_Cubic, leaveTangentWeight=0.1832, tangentMode=RCTM_User, time=0.2082, value=1}
           {interpMode=RCIM_Cubic, tangentMode=RCTM_User, time=0.5, value=1}
           <at 1 0 RCIM_Cubic>
          ]
         }}}
       Scale Curve=0.75
      )
      1=&FloatFromCurve(
       FloatCurve=DataInterface() {propertyValues {Curve {
          defaultValue=1
          keys [
           {arriveTangent=5.7045, interpMode=RCIM_Cubic, leaveTangent=5.7045, tangentMode=RCTM_User, time=0.0358, value=0}
           {arriveTangentWeight=0.1895, interpMode=RCIM_Cubic, leaveTangentWeight=0.1832, tangentMode=RCTM_User, time=0.2082, value=1}
           {interpMode=RCIM_Cubic, tangentMode=RCTM_User, time=0.5, value=1}
           <at 1 0 RCIM_Cubic>
          ]
         }}}
       Scale Curve=0.75
      )
      2=&FloatFromCurve(
       FloatCurve=DataInterface() {propertyValues {Curve {
          defaultValue=1
          keys [
           {arriveTangent=5.7045, interpMode=RCIM_Cubic, leaveTangent=5.7045, tangentMode=RCTM_User, time=0.0512, value=0}
           {arriveTangentWeight=0.1895, interpMode=RCIM_Cubic, leaveTangentWeight=0.1832, tangentMode=RCTM_User, time=0.2901, value=1}
           {interpMode=RCIM_Cubic, tangentMode=RCTM_User, time=0.5, value=1}
           <at 1 0 RCIM_Cubic>
          ]
         }}}
       Scale Curve=0.5
      )
      3=&FloatFromCurve(
       FloatCurve=DataInterface() {propertyValues {Curve {
          defaultValue=1
          keys [
           {arriveTangent=5.7045, interpMode=RCIM_Cubic, leaveTangent=5.7045, tangentMode=RCTM_User, time=0.0512, value=0}
           {arriveTangentWeight=0.1895, interpMode=RCIM_Cubic, leaveTangentWeight=0.1832, tangentMode=RCTM_User, time=0.2901, value=1}
           {interpMode=RCIM_Cubic, tangentMode=RCTM_User, time=0.5, value=1}
           <at 1 0 RCIM_Cubic>
          ]
         }}}
       Scale Curve=0.5
      )
      4=&FloatFromCurve(
       FloatCurve=DataInterface() {propertyValues {Curve {
          defaultValue=1
          keys [
           {arriveTangent=5.7045, interpMode=RCIM_Cubic, leaveTangent=5.7045, tangentMode=RCTM_User, time=0.0358, value=0}
           {arriveTangentWeight=0.1895, interpMode=RCIM_Cubic, leaveTangentWeight=0.1832, tangentMode=RCTM_User, time=0.2082, value=1}
           {interpMode=RCIM_Cubic, tangentMode=RCTM_User, time=0.5, value=1}
           <at 0.8671 0 RCIM_Cubic>
          ]
         }}}
       Scale Curve=0.75
      )
      5=&FloatFromCurve(
       FloatCurve=DataInterface() {propertyValues {Curve {
          defaultValue=1
          keys [
           {arriveTangent=5.7045, interpMode=RCIM_Cubic, leaveTangent=5.7045, tangentMode=RCTM_User, time=0.0358, value=0}
           {arriveTangentWeight=0.1895, interpMode=RCIM_Cubic, leaveTangentWeight=0.1832, tangentMode=RCTM_User, time=0.2082, value=1}
           {interpMode=RCIM_Cubic, tangentMode=RCTM_User, time=0.5, value=1}
           <at 0.8671 0 RCIM_Cubic>
          ]
         }}}
       Scale Curve=0.75
      )
      Index=$Particles.PartitionParticles.Partition
      Num=::NewEnumerator4
     )
     B=&FN_FrontEndFloatModifier(Gameplay=1, Front End=0.3)
    )
    Scale RGB=&MakeVectorFromLinearColorRGB(Color=&ColorFromCurve(ColorCurve=DataInterface() {propertyValues {
        AlphaCurve {keys [<at 0 1 RCIM_Cubic>]}
        BlueCurve {keys [<at 0.374 0 RCIM_Cubic>,<at 1 0.0175 RCIM_Cubic>]}
        GreenCurve {keys [<at 0.374 0.1969 RCIM_Cubic>,<at 1 0 RCIM_Cubic>]}
        RedCurve {keys [<at 0.374 1 RCIM_Cubic>,<at 1 1 RCIM_Cubic>]}
       }}))
    ScaleA=true
    ScaleRGB=true
   )
  ]
  rendererClasses [&NiagaraSpriteRendererProperties]
  renderers [{rendererClass=&NiagaraSpriteRendererProperties}]
 }
 RedGlows {
  emitterUpdateScript [
   &EmitterState(
    Inactive Response=::NewEnumerator1
    Life Cycle Mode=::NewEnumerator1
    Loop Behavior=::NewEnumerator1
    Loop Delay=0.8
    Loop Duration=2
    UseLoopDelay=true
   )
   &SpawnBurst_Instantaneous(Spawn Count=4, Use Spawn Probability=false, UseLoopCountLimit=false)
  ]
  particleSpawnScript [
   &PartitionParticles(Number Of Partitions=4, Particles Per Partition=1, Partition Mode:ENiagara_PartitionMode_v2=::NewEnumerator1)
   &InitializeParticle(
    Color=<rgba 1 0.175 0.0975 1>
    Sprite Size=&MergeHelper_Vector2Switch(0=<70 150>, 1=<70 150>, 2=<70 120>, 3=<70 120>, 4=<0 0>, 5=<0 0>, Index=$Particles.PartitionParticles.Partition, Num=::NewEnumerator4)
    Sprite Size Mode=::NewEnumerator4
    UsePositionOffset=false
    Write Material Random=true
   )
   &DynamicMaterialParameters(
    Index 0 Param 1=3
    Index 0 Param 2=10
    Module.Index0Float0Write=true
    Module.Index0Float1Write=true
    Module.Index0Float2Write=true
    Module.Index0Float3Write=true
   )
   &SocketLocation(
    SamplePosOffset=true
    Sampled Position Offset=<0 -15 -7>
    Socket Reader=DataInterface() {propertyValues {
      AttachComponentClass=null
      AttachComponentTag=None
      EditorPreviewAsset=&YuzuCrank_Glider
      SourceActorRef=null
      SourceAsset=null
      SourceMode=Default
      bRequireCurrentFrameData=true
      bUpdateSocketsPerFrame=true
      FilteredSockets [Body_Offset]
      ObjectParameterBinding {parameter {name=None, typeDefHandle {registeredTypeIndex=1}, varData []}}
     }}
    Socket Sampling Mode=::NewEnumerator3
    VelocitySpeedLimit=false
   )
  ]
  particleUpdateScript [
   &ParticleState()
   &SpriteFacingAndAlignment(
    Alignment Coordinate Space=::Local
    Sprite Alignment=&FN_FrontEndVector3Modifier(
     Gameplay=&MergeHelper_VectorSwitch(
      0=<1 0.684 -0.1962>
      1=<1 -0.684 0.1962>
      2=<1 0.684 -1.1164>
      3=<1 -0.684 1.1164>
      4=<0 0 0>
      5=<0 0 0>
      Index=$Particles.PartitionParticles.Partition
      Num=::NewEnumerator4
     )
     Front End=&MergeHelper_VectorSwitch(
      0=<1 -0.684 0.1962>
      1=<1 0.684 -0.1962>
      2=<1 0.684 -1.1164>
      3=<1 -0.684 1.1164>
      Index=$Particles.PartitionParticles.Partition
      Num=::NewEnumerator2
     )
    )
    WriteAlignment=true
    WriteFacing=false
   )
   &SystemLocation(
    MaskSpawnGroup=false
    Offset=&FN_FrontEndVector3Modifier(
     Gameplay=&MergeHelper_VectorSwitch(0=<160 60 40>, 1=<-160 60 40>, 2=<160 40 -40>, 3=<-160 40 -40>, Index=$Particles.PartitionParticles.Partition, Num=::NewEnumerator2)
     Front End=&MergeHelper_VectorSwitch(0=<160 0 -50>, 1=<-160 0 -50>, 2=<160 0 -70>, 3=<-160 0 -70>, Index=$Particles.PartitionParticles.Partition, Num=::NewEnumerator2)
    )
   )
   &ScaleSpriteSize(Scale Sprite Size Mode=::NewEnumerator0, Uniform Scale Factor=2)
   &ScaleColor(
    Scale Alpha=&FloatFromCurve(
     FloatCurve=DataInterface() {propertyValues {Curve {
        defaultValue=1
        keys [
         {arriveTangent=1.7606, interpMode=RCIM_Cubic, leaveTangent=1.7606, tangentMode=RCTM_User, time=0.0358, value=0}
         {arriveTangentWeight=0.1895, interpMode=RCIM_Cubic, leaveTangentWeight=0.1832, tangentMode=RCTM_User, time=0.4592, value=1}
         {interpMode=RCIM_Cubic, tangentMode=RCTM_User, time=0.5743, value=1}
         <at 0.8671 0 RCIM_Cubic>
        ]
       }}}
     Scale Curve=&FNTimeOfDayScalar(Night Scale=0.2, Morning Scale=0.3, Day Scale=0.3, Evening Scale=0.3)
    )
    ScaleA=true
    ScaleRGB=true
   )
  ]
  rendererClasses [&NiagaraSpriteRendererProperties]
  renderers [{rendererClass=&NiagaraSpriteRendererProperties}]
 }
}
summary {
 systemName=NS_Glider_YuzuCrank_Deploy
 emitters [
  {emitterName=Glows, rendererClasses [&NiagaraSpriteRendererProperties]}
  {emitterName=Frame, rendererClasses [&NiagaraMeshRendererProperties]}
  {emitterName=FrameGlow, rendererClasses [&NiagaraSpriteRendererProperties]}
  {emitterName=RedGlows, rendererClasses [&NiagaraSpriteRendererProperties]}
 ]
 userVariables [
  User.bIsFrontEnd {type=NiagaraBool}
  User.bIsFrontEndPreview {type=NiagaraBool}
 ]
}
