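# Thruster VFX for the YuzuCrank glider, with lens flares, air glow, star flares, mesh trails and speedlines.
# Socket-attached emitters split particles into four partitions that pick thruster sockets and per-thruster offsets, with front-end variants.
# A system timeline ramps size and brightness, time of day dims alpha, and ForwardDot gates speedlines and star flares.
# Trail meshes receive an interpolated guide position, swayed by a smoothed system value, through dynamic material parameters.
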NiagaraSystem /BRCosmetics/Gadgets/Assets/VinderTech_GliderChute/Glider_YuzuCrank/FX/NS_Glider_YuzuCrank_Thrusters
dependencies {
 usedDataInterfaces [
  &NiagaraDataInterfaceMeshRendererInfo
  &NiagaraDataInterfaceArrayInt32
  &NiagaraDataInterfaceCurve
  &NiagaraDataInterfaceVector2DCurve
  &NiagaraDataInterfaceColorCurve
  &NiagaraDataInterfaceSocketReader
  &NiagaraDataInterfaceArrayFloat
 ]
 usedDynamicInputs [
  &SimulationPosition
  &MakeCustomVectorFromBool
  &MakeFloatFromVector
  &SelectFloatFromArray
  &MakeIntFromFloat
  &NormalizeVector
  &Add_Vector
 ]
 usedModules [
  &self:SystemSpawnScript.SystemScriptSource.SystemScriptGraph.NiagaraNodeAssignment_0.SetVariables_#1
  &SystemState
  &FN_Integrator
 ]
}
emitters {
 AirGlow {
  emitterUpdateScript [
   &EmitterState(
    Inactive Response=::NewEnumerator1
    Life Cycle Mode=::NewEnumerator1
    Loop Behavior=::NewEnumerator1
    Loop Delay=&FN_FrontEndFloatModifier(Gameplay=0.5, Front End=0)
    Loop Duration Mode=::NewEnumerator1
    UseLoopDelay=true
   )
   &SpawnBurst_Instantaneous(
    Spawn Count=&MakeCustomIntFromBool(Bool=$User.bIsFrontEndPreview, True INT=0, False INT=4)
    Spawn Group=1
    Use Spawn Probability=false
    UseLoopCountLimit=false
   ) {module=SpawnBurst_Instantaneous001}
   &SpawnBurst_Instantaneous(Spawn Count=4, Use Spawn Probability=false)
  ]
  particleSpawnScript [
   /SetVariables(Particles.VisibilityTag=&SpawnGroup()) {isSetParameters=true}
   &PartitionParticles(Number Of Partitions=4)
   &InitializeParticle(
    Color=&MergeHelper_ColorSwitch(
     0=<rgba 2 0.2914 0 0.7585>
     1=<rgba 2 0.2914 0 0.4879>
     2=<rgba 2 0.5624 0 1>
     3=?
     Index=&SpawnGroup()
     Num=::NewEnumerator1
    )
    Sprite Size=&MergeHelper_Vector2Switch(0=<150 90>, 1=<175 350>, 2=<125 20>, 3=?, 4=?, Index=&SpawnGroup(), Num=::NewEnumerator1)
    Sprite Size Mode=::NewEnumerator4
    UsePositionOffset=false
    Write Material Random=true
   ) {module=InitializeParticle001}
   &DynamicMaterialParameters(
    Index 0 Param 1=2
    Index 0 Param 2=10
    Module.Index0Float0Write=true
    Module.Index0Float1Write=true
    Module.Index0Float2Write=true
    Module.Index0Float3Write=true
   )
   &CameraOffset(Camera Offset Amount=&MergeHelper_FloatSwitch(0=0, 1=20, 2=0, 3=?, Index=&SpawnGroup(), Num=::NewEnumerator1))
   &SpriteFacingAndAlignment(Sprite Facing=<0 0 1>, WriteAlignment=true, WriteFacing=true)
   /SetVariables(
    Particles.Alignment:Vector3f=&MergeHelper_VectorSwitch(0=<0 0 1>, 1=<0 0 1>, 2=<-0.532 0 1>, 3=<0.532 0 1>, Index=$Particles.PartitionParticles.Partition, Num=::NewEnumerator2)
    Particles.Facing:Vector3f=<0 1 0>
   ) {isSetParameters=true}
   /SetVariables(Particles.Rotation:NiagaraFloat=&MergeHelper_FloatSwitch(0=0, 1=0, 2=-24, 3=24, Index=$Particles.PartitionParticles.Partition, Num=::NewEnumerator2)) {isSetParameters=true}
  ]
  particleUpdateScript [
   &ParticleState(Kill Particles When Lifetime Has Elapsed=false, Let Infinitely Lived Particles Die When Emitter Deactivates=true)
   &SocketLocation(
    SamplePosOffset=true
    Sampled Position Offset=&MergeHelper_VectorSwitch(
     0=&FN_FrontEndVector3Modifier(
      Gameplay=&MergeHelper_VectorSwitch(0=<0 30 -5>, 1=<0 30 -5>, 2=<0 -10 0>, 3=<0 10 0>, Index=$Particles.PartitionParticles.Partition, Num=::NewEnumerator2)
      Front End=&MergeHelper_VectorSwitch(0=<0 60 -5>, 1=<0 60 -5>, 2=<50 -10 0>, 3=<-50 10 0>, Index=$Particles.PartitionParticles.Partition, Num=::NewEnumerator2)
     )
     1=&MergeHelper_VectorSwitch(0=<0 110 10>, 1=<0 100 10>, 2=<70 -15 0>, 3=<-70 15 0>, Index=$Particles.PartitionParticles.Partition, Num=::NewEnumerator2)
     2=&MergeHelper_VectorSwitch(0=<0 5 0>, 1=<0 5 0>, 2=<0 -1 0>, 3=<0 0 0>, Index=$Particles.PartitionParticles.Partition, Num=::NewEnumerator2)
     3=?
     Index=&SpawnGroup()
     Num=::NewEnumerator1
    )
    Socket Index=&SelectIntFromArray(Direct Array Index=$Particles.PartitionParticles.Partition, Int Selection Array=DataInterface() {propertyValues {IntData [4,5,6,7]}})
    Socket Reader=$System.Socket Reader
    Socket Sampling Mode=::NewEnumerator3
    VelocitySpeedLimit=false
   )
   &ScaleSpriteSize(Scale Factor=&Vector2DFromFloat(Value=1))
   &ScaleColor(
    Scale Alpha=&Multiply_Float(
     A=&FloatFromCurve(FloatCurve=DataInterface() {propertyValues {Curve {keys [<at 0 0>,<at 0.057 2.5474>,<at 0.1667 1>]}}})
     B=&Multiply_Float(
      A=&ClampFloat(Float=&FN_FrontEndFloatModifier(Gameplay=$System.Timeline.NormalizedTime, Front End=0.4))
      B=&Multiply_Float(
       A=&FNTimeOfDayScalar(Night Scale=0.6, Morning Scale=0.7, Evening Scale=0.7)
       B=&MergeHelper_FloatSwitch(0=0.7, 1=1, 2=1, 3=?, Index=&SpawnGroup(), Num=::NewEnumerator1)
      )
     )
    )
    ScaleA=true
    ScaleRGB=true
   ) {module=ScaleColor001}
  ]
  rendererClasses [&NiagaraSpriteRendererProperties]
  renderers [
   {rendererClass=&NiagaraSpriteRendererProperties}
   {rendererClass=&NiagaraSpriteRendererProperties, rendererIndex=1}
  ]
 }
 LensFlares {
  emitterUpdateScript [
   &EmitterState(
    Life Cycle Mode=::NewEnumerator1
    Loop Behavior=::NewEnumerator1
    Loop Delay=&FN_FrontEndFloatModifier(Gameplay=0.5, Front End=0)
    Loop Duration Mode=::NewEnumerator1
    UseLoopDelay=true
   )
   &SpawnBurst_Instantaneous(Spawn Count=4, Spawn Group=1, Use Spawn Probability=false, UseLoopCountLimit=false) {module=SpawnBurst_Instantaneous001}
   &SpawnBurst_Instantaneous(Spawn Count=4, Use Spawn Probability=false, UseLoopCountLimit=false)
  ]
  particleSpawnScript [
   /SetVariables(Particles.VisibilityTag=&SpawnGroup()) {isSetParameters=true}
   &PartitionParticles(Number Of Partitions=4, Particles Per Partition=1, Partition Mode:ENiagara_PartitionMode_v2=::NewEnumerator1)
   &InitializeParticle(
    Color Mode=::NewEnumerator0
    Mesh Renderer Array Visibility Mode=::NewEnumerator1
    Mesh Renderer Info=DataInterface() {propertyValues {MeshRenderer=&self:LensFlares_0.NiagaraMeshRendererProperties_0}}
    Mesh Renderer Mesh Index:NiagaraInt32=$Particles.PartitionParticles.Partition
    Mesh Scale Mode=::NewEnumerator3
    Sprite Size Mode=::NewEnumerator3
    Uniform Sprite Size=150
    UsePositionOffset=false
    Write Material Random=true
   )
   &CameraOffset(Camera Offset Amount=20)
   &InitialMeshOrientation(
    Mesh Axis To Orient=<0 1 0>
    Mesh Orientation Mode=::NewEnumerator2
    Orientation Vector=&FN_FrontEndVector3Modifier(Gameplay=<0 1 0>, Front End=<0 -1 0>)
    Rotation=&FN_FrontEndVector3Modifier(
     Gameplay=<0 0 0>
     Front End=&MergeHelper_VectorSwitch(0=<0 0 0>, 1=<0 0 0>, 2=<-0.055 0 0>, 3=<-0.055 0 0>, Index=$Particles.PartitionParticles.Partition, Num=::NewEnumerator2)
    )
    Use Rotation Vector=true
   )
  ]
  particleUpdateScript [
   &ParticleState(Kill Particles When Lifetime Has Elapsed=false, Let Infinitely Lived Particles Die When Emitter Deactivates=true)
   &SocketLocation(
    Orientation Sampling=::NewEnumerator1
    SamplePosOffset=true
    Sampled Position Offset=&MaskVectorBySpawnGroup(
     Masked Vector=&FN_FrontEndVector3Modifier(
      Gameplay=&MergeHelper_VectorSwitch(0=<0 5 0>, 1=<0 5 0>, 2=<-2 50 -15>, 3=<2 50 -15>, Index=$Particles.PartitionParticles.Partition, Num=::NewEnumerator2)
      Front End=&MergeHelper_VectorSwitch(0=<0 -5 0>, 1=<0 -5 0>, 2=<-2 -60 -25>, 3=<2 -60 -25>, Index=$Particles.PartitionParticles.Partition, Num=::NewEnumerator2)
     )
     Passthrough Vector=&MergeHelper_VectorSwitch(0=<28 20 -58>, 1=<-30 20 -60>, 2=<22 55 -37>, 3=<-22 55 -37>, Index=$Particles.PartitionParticles.Partition, Num=::NewEnumerator2)
    )
    Sampled Position Offset Coordinate Space:ENiagaraSkelMeshTransforms=::NewEnumerator2
    Socket Index=&SelectIntFromArray(Direct Array Index=$Particles.PartitionParticles.Partition, Int Selection Array=DataInterface() {propertyValues {IntData [0,1,2,3]}})
    Socket Reader=$System.Socket Reader
    Socket Sampling Mode=::NewEnumerator3
    VelocitySpeedLimit=false
   )
   &ScaleMeshSize(Scale Factor=&Lerp_Vector(A=<1 0.5 1>, B=<1 1.5 1>, Alpha=$System.Timeline.NormalizedTime))
   &ScaleSpriteSize(
    Uniform Curve Scale=&FN_FrontEndFloatModifier(Gameplay=&Lerp_Float(A=0.2, B=1.5, Alpha=$System.Timeline.NormalizedTime), Front End=1.5)
    Uniform Curve Sprite Scale=DataInterface() {propertyValues {Curve {
       defaultValue=0
       keys [
        {interpMode=RCIM_Cubic, tangentMode=RCTM_User, time=0, value=0.6}
        {interpMode=RCIM_Cubic, tangentMode=RCTM_User, time=0.0533, value=1.2}
        <at 0.2198 1 RCIM_Cubic>
       ]
      }}}
   )
   &ScaleColor(
    Scale Alpha=&Multiply_Float(
     A=&Lerp_Float(A=0.5, B=1, Alpha=$System.Timeline.NormalizedTime)
     B=&FNTimeOfDayScalar(Night Scale=0.6, Morning Scale=0.8, Evening Scale=0.8)
    )
    Scale RGB=&Lerp_Vector(
     A=&MakeVectorFromLinearColorRGB(Color=<rgba 1 0.0232 0.0001 1>)
     B=&MakeVectorFromLinearColorRGB(Color=<rgba 5 0.1162 0.0005 1>)
     Alpha=$System.Timeline.NormalizedTime
    )
    ScaleA=true
    ScaleRGB=true
   )
   /SetVariables(Particles.MeshColor:LinearColor=&MakeLinearColorFromVectorAndFloat(
     "Vector (RGB)"=&MakeVectorFromLinearColorRGB(Color=<rgba 5 0.0346 0 0.5807>)
     "Float (Alpha)"=&FloatFromCurve(FloatCurve=DataInterface() {propertyValues {Curve {
         defaultValue=1
         keys [
          {arriveTangent=4.1785, interpMode=RCIM_Cubic, leaveTangent=4.1785, tangentMode=RCTM_User, time=0, value=0}
          {arriveTangent=1.1179, interpMode=RCIM_Cubic, leaveTangent=1.1179, tangentMode=RCTM_User, time=0.3082, value=1.0586}
          {arriveTangent=0.0224, interpMode=RCIM_Cubic, leaveTangent=0.0224, tangentMode=RCTM_User, time=1, value=1}
         ]
        }}})
    )) {isSetParameters=true}
   &SystemLocation(
    MaskSpawnGroup=false
    Offset=&MaskVectorBySpawnGroup(
     Masked Vector=Linked() {linkedVariable {name=StackContext.Position, type=NiagaraPosition}}
     Passthrough Vector=&FN_FrontEndVector3Modifier(
      Gameplay=&Lerp_Vector(
       A=&AddVectorToPosition(
        Position=$Output.SocketLocation.SampledPosition
        Vector=&MergeHelper_VectorSwitch(0=<0 -10 0>, 1=<0 -10 0>, 2=<0 -5 0>, 3=<0 -5 0>, Index=$Particles.PartitionParticles.Partition, Num=::NewEnumerator2)
       )
       B=&AddVectorToPosition(
        Position=$Output.SocketLocation.SampledPosition
        Vector=&MergeHelper_VectorSwitch(0=<0 10 0>, 1=<0 10 0>, 2=<0 5 0>, 3=<0 5 0>, Index=$Particles.PartitionParticles.Partition, Num=::NewEnumerator2)
       )
       Alpha=$System.Timeline.NormalizedTime
      )
      Front End=&AddVectorToPosition(
       Position=$Output.SocketLocation.SampledPosition
       Vector=&MergeHelper_VectorSwitch(0=<-55 -30 0>, 1=<60 -30 0>, 2=<-45 -99 -1>, 3=<45 -99 -1>, Index=$Particles.PartitionParticles.Partition, Num=::NewEnumerator2)
      )
     )
    )
   )
  ]
  rendererClasses [&NiagaraSpriteRendererProperties, &NiagaraMeshRendererProperties]
  renderers [
   {rendererClass=&NiagaraSpriteRendererProperties}
   {rendererClass=&NiagaraMeshRendererProperties, rendererIndex=1}
  ]
 }
 MeshTrail {
  emitterUpdateScript [
   &EmitterState(
    Inactive Response=::NewEnumerator1
    Life Cycle Mode=::NewEnumerator1
    Loop Behavior=::NewEnumerator1
    Loop Delay=&FN_FrontEndFloatModifier(Gameplay=0.5, Front End=0)
    Loop Duration Mode=::NewEnumerator1
    UseLoopDelay=true
   )
   &SpawnBurst_Instantaneous(Spawn Count=4, Spawn Group=1, Use Spawn Probability=false, UseLoopCountLimit=false) {module=SpawnBurst_Instantaneous001}
   &SpawnBurst_Instantaneous(Spawn Count=4, Use Spawn Probability=false, UseLoopCountLimit=false)
  ]
  particleSpawnScript [
   /SetVariables(Particles.VisibilityTag=&SpawnGroup()) {isSetParameters=true}
   &PartitionParticles(Number Of Partitions=4)
   &InitializeParticle(
    Mesh Scale=&FN_FrontEndVector3Modifier(Gameplay=<3.5 1 0.515>, Front End=<3.5 1 0.515>)
    Mesh Scale Mode=::NewEnumerator4
    UsePositionOffset=false
    Write Material Random=true
   )
   &InitialMeshOrientation(
    Orientation Vector=<0 0 1>
    Rotation=&MergeHelper_VectorSwitch(0=<0 0 0.25>, 1=<0 0 0.25>, 2=<0 0.25 0>, 3=<0 0.75 0>, Index=$Particles.PartitionParticles.Partition, Num=::NewEnumerator2)
    Use Orientation Vector=true
    Use Rotation Vector=true
   )
   &InitialMeshOrientation(
    Orientation Vector=<0 0 1>
    Rotation=&MergeHelper_VectorSwitch(0=<0 0 0>, 1=<0 0 0>, 2=<0 0.02 0>, 3=<0 0.02 0>, Index=$Particles.PartitionParticles.Partition, Num=::NewEnumerator2)
    Use Orientation Vector=false
    Use Rotation Vector=true
   ) {module=InitialMeshOrientation001}
   &CameraOffset(Camera Offset Amount=1)
   /SetVariables(Particles.Random=&RandomRangeFloat(Minimum=0, Maximum=1)) {isSetParameters=true}
  ]
  particleUpdateScript [
   &ParticleState(Kill Particles When Lifetime Has Elapsed=false, Let Infinitely Lived Particles Die When Emitter Deactivates=true)
   &SocketLocation(
    SamplePosOffset=true
    Sampled Position Offset=&FN_FrontEndVector3Modifier(
     Gameplay=<0 0 0>
     Front End=&MergeHelper_VectorSwitch(0=<0 0 2>, 1=<0 0 2>, 2=<-1 1 0>, 3=<1 -1 0>, Index=$Particles.PartitionParticles.Partition, Num=::NewEnumerator2)
    )
    Socket Index=&SelectIntFromArray(Direct Array Index=$Particles.PartitionParticles.Partition, Int Selection Array=DataInterface() {propertyValues {IntData [4,5,6,7]}})
    Socket Reader=$System.Socket Reader
    Socket Sampling Mode=::NewEnumerator3
    VelocitySpeedLimit=false
   )
   /SetVariables(Particles.GuidePosition:NiagaraPosition=&AddVectorToPosition(
     Position=$Output.SocketLocation.SampledPosition
     Vector=&FN_FrontEndVector3Modifier(
      Gameplay=&TransformVector(
       Vector=&Lerp_Vector(
        A=<-700 200 0>
        B=<700 200 300>
        Alpha=&RemapRange(
         Input Value To Remap=$System.InterpolateOverTime001.InterpolateOverTime_Float.Moving Average
         Input Min=-1
         Input Max=1
         Output Min=0
         Output Max=1
        )
       )
       Destination Space=::World
      )
      Front End=&TransformVector(Vector=<0 0 0>, Source Space=::Simulation, Destination Space=::World)
     )
    )) {isSetParameters=true}
   &InterpolateOverTime(
    Rate Of Change:NiagaraFloat=&FN_FrontEndFloatModifier(Gameplay=10, Front End=4)
    Target Vector Value:NiagaraPosition=$Particles.GuidePosition
    Variable Type=::NewEnumerator2
   )
   &DynamicMaterialParameters(
    Index 0 Param 2=0.2276
    Index 1 Param 1=$Particles.Random
    Index 1 Param 3=5
    Index 2 Param 4=&FN_FrontEndFloatModifier(Gameplay=1, Front End=1)
    Index 2 Vector 1:Vector3f=$Output.InterpolateOverTime.Moving Average
    Module.Index0Float0Write=true
    Module.Index0Float1Write=true
    Module.Index0Float2Write=true
    Module.Index0Float3Write=true
    Module.Index1Float0Write=true
    Module.Index1Float1Write=true
    Module.Index1Float2Write=true
    Module.Index1Float3Write=true
    Module.Index2Float3Write=true
    Module.Index2VectorWrite=true
    Write Parameter Index 1=::NewEnumerator1
    Write Parameter Index 2=::NewEnumerator3
   )
   &ScaleColor(
    Scale Alpha=&Multiply_Float(
     A=&FloatFromCurve(FloatCurve=DataInterface() {propertyValues {Curve {keys [<at 0 0>,<at 0.2 1>,<at 1 1>]}}})
     B=&Multiply_Float(
      A=&ClampFloat(Float=&FN_FrontEndFloatModifier(Gameplay=$System.Timeline.NormalizedTime, Front End=0.3))
      B=&FNTimeOfDayScalar(Night Scale=0.6, Morning Scale=0.7, Evening Scale=0.7)
     )
    )
    Scale RGB=&MaskVectorBySpawnGroup(Masked Vector=<2 2 2>, Passthrough Vector=<1 1 1>)
    ScaleA=true
    ScaleRGB=true
   )
   &ScaleMeshSize(Scale Factor=&FN_FrontEndVector3Modifier(
     Gameplay=&Lerp_Vector(A=<0 0 0.8>, B=<1 1 0.8>, Alpha=$System.Timeline.NormalizedTime)
     Front End=&MergeHelper_VectorSwitch(0=<1 1 0.5>, 1=<1 1 0.5>, 2=<1 1 0.5>, 3=<1 1 0.5>, Index=$Particles.PartitionParticles.Partition, Num=::NewEnumerator2)
    ))
   /SetVariables(
    Particles.Background Dynamic Parameter:Vector4f=&MakeVector4(X=0, Y=0, Z=0, W=0)
    Particles.BackgroundSize:Vector3f=&FN_FrontEndVector3Modifier(
     Gameplay=&Lerp_Vector(A=<0 0 0.75>, B=<3.5 0.1 0.75>, Alpha=$System.Timeline.NormalizedTime)
     Front End=&MergeHelper_VectorSwitch(0=<3.5 0.1 0.5>, 1=<3.5 0.1 0.5>, 2=<3.5 0.1 0.4>, 3=<3.5 0.1 0.4>, Index=$Particles.PartitionParticles.Partition, Num=::NewEnumerator2)
    )
   ) {isSetParameters=true}
  ]
  rendererClasses [&NiagaraMeshRendererProperties]
  renderers [
   {rendererClass=&NiagaraMeshRendererProperties, rendererIndex=0}
   {rendererClass=&NiagaraMeshRendererProperties, rendererIndex=1}
  ]
 }
 Speedlines001 {
  emitterUpdateScript [
   &EmitterState()
   &SpawnRate(
    Spawn Probability=&MakeCustomFloatFromBool(
     Bool=&SetBoolByFloatComparison(A=$User.ForwardDot, B=0.05)
     True Float=1
     False Float=&FN_FrontEndFloatModifier(Gameplay=0, Front End=1)
    )
    SpawnRate=&MakeCustomFloatFromBool(Bool=$User.bIsFrontEndPreview, True Float=50, False Float=90)
    Use Spawn Probability=true
   ) {module=SpawnRate001}
  ]
  particleSpawnScript [
   &InitializeParticle(
    Color=&MakeLinearColorFromVectorAndFloat(
     "Vector (RGB)"=<1 1 1>
     "Float (Alpha)"=&Multiply_Float(
      A=&UniformRangedFloat(Minimum=1, Maximum=1)
      B=&MakeCustomFloatFromBool(
       Bool=$User.bIsFrontEndPreview
       True Float=&FloatFromCurve(
        FloatCurve=DataInterface() {propertyValues {Curve {keys [{tangentMode=RCTM_User, time=0, value=0}, {tangentMode=RCTM_User, time=2, value=1}]}}}
        CurveIndex=$Emitter.Age
        Scale Curve=0.4
       )
       False Float=1
      )
     )
    )
    Lifetime Max=0.2
    Lifetime Min=0.15
    Lifetime Mode=::NewEnumerator1
    Position Offset=&FN_FrontEndVector3Modifier(Gameplay=<0 -50 0>, Front End=<0 50 0>)
    Sprite Size Max=<3 10>
    Sprite Size Min=<2 2>
    Sprite Size Mode=::NewEnumerator2
    UsePositionOffset=true
    Write Material Random=true
   )
   &AddVelocity(
    Module.Use Falloff Distance=false
    Override Seed=false
    Velocity Mode=::NewEnumerator1
    Velocity Speed=&FN_FrontEndFloatModifier(Gameplay=-2500, Front End=-2500)
   ) {module=AddVelocity001}
   &TorusLocation(
    Align Torus With Axis=true
    Coordinate Space:ENiagaraCoordinateSpace=::Simulation
    Handle Radius=0
    Large Radius:NiagaraFloat=&FN_FrontEndFloatModifier(
     Gameplay=&RandomRangeFloat(Minimum=450, Maximum=500)
     Front End=&RandomRangeFloat(Minimum=350, Maximum=400)
    )
    MaskSpawnGroup=false
    Module.Override Randomness=false
    Offset=<0 33 0>
    Torus Axis=<0 1 0>
    Torus Origin:NiagaraPosition=&AddVectorToPosition(
     Position=$Particles.Position
     Vector=&FN_FrontEndVector3Modifier(Gameplay=<0 -500 0>, Front End=<0 500 0>)
    )
   )
  ]
  particleUpdateScript [
   &SolveForcesAndVelocity(Clamp Velocity=false, Limit Acceleration=false)
   &ParticleState()
   &ScaleSpriteSize(Scale Factor=&Vector2DFromCurve(
     Vector2Curve=DataInterface() {propertyValues {XCurve {keys [<at 0 0.5 RCIM_Cubic>,<at 0.5 1 RCIM_Cubic>]}, YCurve {keys [<at 0 0 RCIM_Cubic>,<at 0.5 1 RCIM_Cubic>]}}}
     ScaleCurve=&Vector2DFromFloat(Value=1)
    ))
   &ScaleSpriteSizeBySpeed(Max Scale Factor=<1 50>, Min Scale Factor=<0.01 0.05>, Velocity Threshold=1)
   &ScaleColor(
    Scale Alpha=&FloatFromCurve(
     FloatCurve=DataInterface() {propertyValues {Curve {keys [
         <at 0 0 RCIM_Cubic>
         <at 0.3759 1.0207>
         {arriveTangent=-2.221, interpMode=RCIM_Cubic, leaveTangent=-2.221, tangentMode=RCTM_User, time=1, value=0}
        ]}}}
     Scale Curve=&MakeCustomFloatFromBool(
      Bool=$User.bIsFrontEndPreview
      True Float=0.3759
      False Float=&Lerp_Float(A=0, B=0.5, Alpha=&FN_FrontEndFloatModifier(Gameplay=$System.Timeline.Time, Front End=0.5))
     )
    )
    Scale RGB=&VectorFromFloat(Value=&FN_FrontEndFloatModifier(Gameplay=0.5, Front End=0.33))
    ScaleA=true
    ScaleRGB=true
   )
  ]
  rendererClasses [&NiagaraSpriteRendererProperties]
  renderers [{rendererClass=&NiagaraSpriteRendererProperties}]
 }
 StarFlare {
  emitterUpdateScript [
   &EmitterState(
    Life Cycle Mode=::NewEnumerator1
    Loop Behavior=::NewEnumerator1
    Loop Delay=&FN_FrontEndFloatModifier(Gameplay=0.5, Front End=0)
    Loop Duration Mode=::NewEnumerator1
    UseLoopDelay=true
   )
   &SpawnBurst_Instantaneous(Spawn Count=4, Use Spawn Probability=false, UseLoopCountLimit=false)
  ]
  particleSpawnScript [
   &InitializeParticle(
    Color=<rgba 3 1.2657 0.2399 1>
    Sprite Size Mode=::NewEnumerator3
    Uniform Sprite Size=500
    UsePositionOffset=false
    Write Material Random=true
   )
   &PartitionParticles(Number Of Partitions=4)
   &SocketLocation(
    SamplePosOffset=true
    Sampled Position Offset=&MergeHelper_VectorSwitch(0=<0 40 10>, 1=<0 40 10>, 2=<30 0 0>, 3=<-30 0 0>, Index=$Particles.PartitionParticles.Partition, Num=::NewEnumerator2)
    Socket Index=&SelectIntFromArray(Direct Array Index=$Particles.PartitionParticles.Partition, Int Selection Array=DataInterface() {propertyValues {IntData [4,5,6,7]}})
    Socket Reader=$System.Socket Reader
    Socket Sampling Mode=::NewEnumerator3
    VelocitySpeedLimit=false
   )
   &CameraOffset(Camera Offset Amount=25)
  ]
  particleUpdateScript [
   &ParticleState(Kill Particles When Lifetime Has Elapsed=false, Let Infinitely Lived Particles Die When Emitter Deactivates=true)
   &Timeline(
    Instantaneous Rewind=true
    Play=&SetBoolByFloatComparison(A=$User.ForwardDot, B=0.5)
    Play Rate=8
    Rewind Rate=0
   )
   &ScaleSpriteSize(
    Uniform Curve Index=$Particles.Timeline.InverseNormalizedTime
    Uniform Curve Sprite Scale=DataInterface() {propertyValues {Curve {
       defaultValue=0
       keys [
        {interpMode=RCIM_Cubic, tangentMode=RCTM_User, time=0, value=0}
        {interpMode=RCIM_Cubic, tangentMode=RCTM_User, time=0.5, value=1}
        {arriveTangent=-4.9166, interpMode=RCIM_Cubic, leaveTangent=-4.9166, tangentMode=RCTM_User, time=1, value=0}
       ]
      }}}
   )
   &SolveForcesAndVelocity(Clamp Velocity=false, Limit Acceleration=false)
   &ScaleColor(
    Scale Alpha=&FNTimeOfDayScalar(Night Scale=0.6, Morning Scale=0.8, Evening Scale=0.8)
    ScaleA=true
    ScaleRGB=false
   )
  ]
  rendererClasses [&NiagaraSpriteRendererProperties]
  renderers [{rendererClass=&NiagaraSpriteRendererProperties}]
 }
 Trails {
  emitterUpdateScript [
   &EmitterState(
    Inactive Response=::NewEnumerator1
    Life Cycle Mode=::NewEnumerator1
    Loop Behavior=::NewEnumerator1
    Loop Delay=&FN_FrontEndFloatModifier(Gameplay=0.5, Front End=0)
    Loop Duration Mode=::NewEnumerator1
    UseLoopDelay=true
   )
   &SpawnBurst_Instantaneous(Spawn Count=4, Use Spawn Probability=false)
  ]
  particleSpawnScript [
   &PartitionParticles(Number Of Partitions=4)
   &InitializeParticle(
    Color=<rgba 10 0.8355 0.4365 1>
    Mesh Scale=<3.5 1 0.65>
    Mesh Scale Mode=::NewEnumerator4
    UsePositionOffset=false
    Write Material Random=true
   ) {module=InitializeParticle001}
   &InitialMeshOrientation(
    Orientation Vector=<0 0 1>
    Rotation=&MergeHelper_VectorSwitch(0=<0 0 0.25>, 1=<0 0 0.25>, 2=<0 0.25 0>, 3=<0 0.75 0>, Index=$Particles.PartitionParticles.Partition, Num=::NewEnumerator2)
    Use Orientation Vector=true
    Use Rotation Vector=true
   )
   &InitialMeshOrientation(
    Orientation Vector=<0 0 1>
    Rotation=&MergeHelper_VectorSwitch(0=<0 0 0>, 1=<0 0 0>, 2=<0 0.02 0>, 3=<0 0.02 0>, Index=$Particles.PartitionParticles.Partition, Num=::NewEnumerator2)
    Use Orientation Vector=false
    Use Rotation Vector=true
   ) {module=InitialMeshOrientation001}
   &CameraOffset(Camera Offset Amount=1)
  ]
  particleUpdateScript [
   &ParticleState(Kill Particles When Lifetime Has Elapsed=false, Let Infinitely Lived Particles Die When Emitter Deactivates=true)
   &SocketLocation(
    SamplePosOffset=true
    Sampled Position Offset=&FN_FrontEndVector3Modifier(
     Gameplay=<0 0 0>
     Front End=&MergeHelper_VectorSwitch(0=<0 0 2>, 1=<0 0 2>, 2=<-1 1 0>, 3=<1 -1 0>, Index=$Particles.PartitionParticles.Partition, Num=::NewEnumerator2)
    )
    Socket Index=&SelectIntFromArray(Direct Array Index=$Particles.PartitionParticles.Partition, Int Selection Array=DataInterface() {propertyValues {IntData [4,5,6,7]}})
    Socket Reader=$System.Socket Reader
    Socket Sampling Mode=::NewEnumerator3
    VelocitySpeedLimit=false
   )
   /SetVariables(Particles.GuidePosition:NiagaraPosition=&AddVectorToPosition(
     Position=$Output.SocketLocation.SampledPosition
     Vector=&FN_FrontEndVector3Modifier(
      Gameplay=&TransformVector(
       Vector=&Lerp_Vector(
        A=<-700 200 0>
        B=<700 200 300>
        Alpha=&RemapRange(
         Input Value To Remap=$System.InterpolateOverTime001.InterpolateOverTime_Float.Moving Average
         Input Min=-1
         Input Max=1
         Output Min=0
         Output Max=1
        )
       )
       Destination Space=::World
      )
      Front End=&TransformVector(Vector=<0 0 0>, Source Space=::Simulation, Destination Space=::World)
     )
    )) {isSetParameters=true}
   &InterpolateOverTime(
    Rate Of Change:NiagaraFloat=&FN_FrontEndFloatModifier(Gameplay=10, Front End=4)
    Target Vector Value:NiagaraPosition=$Particles.GuidePosition
    Variable Type=::NewEnumerator2
   )
   /SetVariables(Particles.Interpolate Dynamic=&MakeFloat4FromVectorAndFloat(
     "Vector (RGB)"=$Output.InterpolateOverTime.Moving Average
     "Float (Alpha)"=&FN_FrontEndFloatModifier(Gameplay=1, Front End=1)
    )) {isSetParameters=true}
   /SetVariables(Particles.Mesh_Trail_Color02:LinearColor=&MakeLinearColorFromVectorAndFloat(
     "Vector (RGB)"=&MakeVectorFromLinearColorRGB(Color=<rgba 10 0.0676 0 1>)
     "Float (Alpha)"=&Multiply_Float(
      A=&Multiply_Float(
       A=&FloatFromCurve(FloatCurve=DataInterface() {propertyValues {Curve {keys [<at 0 0>,<at 0.1 1>,<at 1 1>]}}})
       B=&ClampFloat(Float=&FN_FrontEndFloatModifier(Gameplay=$System.Timeline.NormalizedTime, Front End=1))
      )
      B=&ClampFloat(Float=&FN_FrontEndFloatModifier(Gameplay=$System.Timeline.NormalizedTime, Front End=1))
     )
    )) {isSetParameters=true}
   &DynamicMaterialParameters(
    Index 0 Param 1=3
    Index 0 Param 2=0.4
    Module.Index0Float0Write=true
    Module.Index0Float1Write=true
    Module.Index0Float2Write=true
    Module.Index0Float3Write=true
   )
   /SetVariables(Particles.Dynamic02:Vector4f=&MakeVector4(X=1, Y=0.5, Z=0, W=0)) {isSetParameters=true}
   /SetVariables(Particles.Dynamic03:Vector4f=&MakeVector4(X=0.93, Y=0, Z=0, W=0)) {isSetParameters=true}
   &ScaleColor(
    Scale Alpha=&Multiply_Float(
     A=&FloatFromCurve(FloatCurve=DataInterface() {propertyValues {Curve {keys [<at 0 0>,<at 0.2 1>,<at 1 1>]}}})
     B=&Multiply_Float(
      A=&ClampFloat(Float=&FN_FrontEndFloatModifier(Gameplay=$System.Timeline.NormalizedTime, Front End=0.5))
      B=&FNTimeOfDayScalar(Night Scale=0.6, Morning Scale=0.7, Evening Scale=0.7)
     )
    )
    ScaleA=true
    ScaleRGB=false
   )
   &ScaleMeshSize(Scale Factor=&FN_FrontEndVector3Modifier(
     Gameplay=&Lerp_Vector(A=<0 0 0.8>, B=<1 1 0.8>, Alpha=$System.Timeline.NormalizedTime)
     Front End=&MergeHelper_VectorSwitch(0=<1 1 0.5>, 1=<1 1 0.5>, 2=<1 1 0.4>, 3=<1 1 0.4>, Index=$Particles.PartitionParticles.Partition, Num=::NewEnumerator2)
    ))
  ]
  rendererClasses [&NiagaraMeshRendererProperties]
  renderers [
   {rendererClass=&NiagaraMeshRendererProperties, rendererIndex=0}
   {rendererClass=&NiagaraMeshRendererProperties, rendererIndex=1}
  ]
 }
}
summary {
 systemName=NS_Glider_YuzuCrank_Thrusters
 emitters [
  {emitterName=LensFlares, rendererClasses [&NiagaraSpriteRendererProperties, &NiagaraMeshRendererProperties]}
  {emitterName=Trails, rendererClasses [&NiagaraMeshRendererProperties]}
  {emitterName=AirGlow, rendererClasses [&NiagaraSpriteRendererProperties]}
  {emitterName=MeshTrail, rendererClasses [&NiagaraMeshRendererProperties]}
  {emitterName=StarFlare, rendererClasses [&NiagaraSpriteRendererProperties]}
  {emitterName=Speedlines001, rendererClasses [&NiagaraSpriteRendererProperties]}
 ]
 userVariables [
  User.bIsFrontEnd {type=NiagaraBool}
  User.bIsFrontEndPreview {type=NiagaraBool, defaultValue=true}
  User.bIsFullyDeployed {type=NiagaraBool}
  User.ForwardDot {type=NiagaraFloat, defaultValue=1, description=""}
  User.RibbonIDOffset {type=NiagaraInt32}
  User.RightDot {type=NiagaraFloat, defaultValue=0, description=""}
 ]
}
