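# Material function for the YuzuCrank glider that layers an animated "Mainframe" wipe-in effect onto input material attributes, overriding emissive, opacity mask and WPO.
# Timestamp parameters, or a looping test switch, drive staggered outline and main-material fade-ins.
# Local-position sphere-mask wipes, a channel-selected line mask, panning noise and flicker shape dithered opacity and glowing emissive.
# An optional ShadowReplace offset collapses vertices to the object position in shadow passes, hiding shadows.
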MaterialFunction /BRCosmetics/Gadgets/Assets/VinderTech_GliderChute/Glider_YuzuCrank/Materials/MF_YuzuCrank_Glider
graph {nodes [
  Constant()
  FunctionOutput(None=SetMaterialAttributes!(
    MaterialAttributes=NamedRerouteUsage()
    EmissiveColor=NamedRerouteUsage()
    OpacityMask=NamedRerouteUsage()
    WorldPositionOffset=NamedRerouteUsage()
   ) {attributeSetTypes [#1, #2, #3]}) {outputName=Result Material Attributes}
  n1 = ScalarParameter() {defaultValue=10, parameterName=Param_3}
  MaterialFunctionCall!(Time Mult=@n1) {materialFunction=/Game/Effects/Fort_Effects/MaterialFunctions/MF_Flicker}
  n2 = LinearInterpolate(Alpha=NamedRerouteUsage())
  n3 = NamedRerouteUsage()
  Multiply(A=@n2, B=Power(Base=@n3, Exp=ScalarParameter() {defaultValue=1, parameterName=Exp}) {constExponent=1.5})
  n4 = NamedRerouteUsage()
  n5 = Power(Base=NamedRerouteUsage(), Exp=ScalarParameter() {defaultValue=1.25, parameterName=Exp_2}) {constExponent=2.5}
  n6 = NamedRerouteUsage()
  n7 = Add(A=@n5, B=Subtract(A=Max(A=@n6) {constB=0.1}))
  Multiply(A=SmoothStep(Value=@n4) {constMax=0.6, constMin=0.4}, B=Saturate(None=Multiply(A=@n7) {constB=5}))
  n8 = ScalarParameter() {defaultValue=0.5, parameterName=LinesExpMin}
  n9 = ScalarParameter() {defaultValue=1, parameterName=LinesExpMax}
  n10 = Multiply(A=Power(Base=NamedRerouteUsage(), Exp=LinearInterpolate(A=@n8, B=@n9, Alpha=@n3) {constA=1}) {constExponent=1.5}, B=@n2)
  n11 = LocalPosition() {localOrigin=InstancePreSkinning}
  n12 = VectorParameter() {parameterName=FirstWipePosOffset, defaultValue=<rgba 0 0 100 1>}
  n13 = ComponentMask(None=Multiply(A=Add(A=@n11.XYZ, B=@n12.RGB), B=Constant() {r=-0.003})) {b=true, g=true, r=true}
  n14 = Multiply(A=ScalarParameter() {defaultValue=-0.25, parameterName=Param_2}, B=NamedRerouteUsage())
  n15 = Multiply(A=Time(), B=ScalarParameter() {defaultValue=0.4, group=MainframeEffect, parameterName=Speed})
  n16 = Saturate(None=StaticSwitchParameter(True=Frac(None=@n15), False=NamedRerouteUsage()) {group=MainframeEffect, parameterName=TestAnimation})
  n17 = ScalarParameter() {group=MainframeEffectTiming, parameterName=OutlineFadeInStart, sliderMax=1, sortPriority=1}
  n18 = ScalarParameter() {defaultValue=0.7, group=MainframeEffectTiming, parameterName=OutlineFadeInEnd, sliderMax=1, sortPriority=2}
  n19 = Constant()
  n20 = Constant() {r=1}
  n21 = NamedRerouteDeclaration(None=Saturate(None=MaterialFunctionCall(Input=@n16, Input Low=@n17, Input High=@n18, Target Low=@n19, Target High=@n20) {materialFunction=&RemapValueRange}.Result)) {name=FirstFadeIn, nodeColor=<rgba 0 0.1562 1 1>}
  n22 = Max(A=NamedRerouteUsage(), B=SphereMask(A=VertexInterpolator(VS=@n13).PS, B=Constant(), Radius=Add(A=@n14, B=@n21), Hardness=Constant() {r=0.8}))
  n23 = SmoothStep(Value=NamedRerouteUsage()) {constMax=0.6, constMin=0.4}
  n24 = NamedRerouteUsage()
  n25 = Saturate(None=Subtract(A=Multiply(A=@n23) {constB=2}, B=OneMinus(None=@n24)))
  NamedRerouteDeclaration(None=Max(
    A=Multiply(A=@n10, B=@n22)
    B=MaterialFunctionCall(Alpha Threshold=@n25) {materialFunction=&DitherTemporalAA_Fortnite}.Result
   )) {name=OutOpacity, nodeColor=<rgba 0.9688 1 0 1>}
  n26 = Multiply(A=NamedRerouteUsage(), B=@n10)
  n27 = OneMinus(None=NamedRerouteUsage())
  n28 = Power(Base=@n27, Exp=ScalarParameter() {defaultValue=6, parameterName=Exp_1})
  n29 = Min(A=OneMinus(None=NamedRerouteUsage()), B=NamedRerouteUsage())
  n30 = Multiply(A=Add(A=NamedRerouteUsage(), B=@n29), B=@n29)
  n31 = Multiply(A=@n30, B=@n30)
  n32 = ScalarParameter() {defaultValue=15, group=MainframeEffect, parameterName=SecondWipeEmissiveInt}
  n33 = Multiply(A=Add(A=Multiply(A=@n26, B=@n28), B=Multiply(A=@n31, B=@n32)), B=@n27)
  n34 = VectorParameter() {group=MainframeEffect, parameterName=MainFrameEmissive, defaultValue=<rgba 25 0.8227 0 1>}
  NamedRerouteDeclaration(None=Add(A=NamedRerouteUsage(), B=Multiply(A=@n33, B=@n34.RGB))) {name=OutEmissive, nodeColor=<rgba 1 0 0.8672 1>}
  n35 = TextureSample() {samplerType=SAMPLERTYPE_Masks, texture=/BRCosmetics/Gadgets/Assets/VinderTech_GliderChute/Glider_YuzuCrank/Textures/T_YuzuCrank_Glider_FX}
  NamedRerouteDeclaration(None=ChannelMaskParameter(None=FunctionInput(Preview=@n35.RGB) {bUsePreviewValueAsDefault=true, inputName=InFXMask, sortPriority=2}) {parameterName=GliderLinesMask}.R) {name=InLines, nodeColor=<rgba 0.1719 0 1 1>}
  n36 = TextureCoordinate()
  n37 = Add(A=Multiply(A=@n36, B=ScalarParameter() {defaultValue=1.5, parameterName=LineFadeColorTiling}), B=Constant2Vector() {g=0.2651, r=-1.1561})
  NamedRerouteDeclaration(None=TextureSample(UVs=Panner(Coordinate=@n37) {speedX=0.3, speedY=0.25}) {samplerType=SAMPLERTYPE_Masks, texture=/Game/Global/Textures/Default/Noise/WindNoise_Mask_1}.R) {name=GaussianNoiseAnim2, nodeColor=<rgba 1 0 0.8437 1>}
  n38 = ScalarParameter() {defaultValue=1.5, parameterName=B}
  n39 = TextureSample(UVs=Panner(Coordinate=Multiply(A=@n36, B=@n38) {constB=0.5}) {speedX=0.1, speedY=0.05}) {samplerType=SAMPLERTYPE_Masks, texture=&WindNoise_Mask_1}
  NamedRerouteDeclaration(None=@n39.R) {name=HighFreqNoiseAnim, nodeColor=<rgba 0.4297 1 0 1>}
  n40 = Round(None=Sine(None=TextureCoordinate()))
  n41 = ComponentMask(None=Abs(None=ConstantBiasScale(None=@n40))) {r=true}
  NamedRerouteDeclaration(None=Saturate(None=Divide(A=@n41) {constB=3})) {name=RandomIDMask, nodeColor=<rgba 0 0.3906 1 1>}
  n42 = ScalarParameter() {defaultValue=0.45, group=MainframeEffectTiming, parameterName=MainMaterialFadeInStart, sliderMax=1, sortPriority=3}
  n43 = ScalarParameter() {defaultValue=1, group=MainframeEffectTiming, parameterName=MainMaterialFadeInEnd, sliderMax=1, sortPriority=4}
  n44 = Constant()
  n45 = Constant() {r=1}
  NamedRerouteDeclaration(None=Saturate(None=MaterialFunctionCall(Input=@n16, Input Low=@n42, Input High=@n43, Target Low=@n44, Target High=@n45) {materialFunction=&RemapValueRange}.Result)) {name=SecondFadeIn, nodeColor=<rgba 0.7109 1 0 1>}
  n46 = Multiply(A=@n5, B=ScalarParameter() {defaultValue=1.5, parameterName=B_1})
  n47 = Multiply(A=Add(A=Subtract(A=@n4), B=Saturate(None=@n46))) {constB=2}
  n48 = Time()
  n49 = ConstantBiasScale(None=Sine(None=Multiply(A=@n48, B=@n1)))
  NamedRerouteDeclaration(None=Multiply(A=Saturate(None=@n47), B=Max(A=@n49, B=@n5))) {name=Flicker, nodeColor=<rgba 1 0.1641 0 1>}
  n50 = Constant()
  n51 = ObjectPositionWS()
  n52 = WorldPosition()
  n53 = ShadowReplace(
   Default=@n50
   Shadow=StaticSwitchParameter(True=Subtract(A=@n51, B=@n52.XYZ), False=@n50) {defaultValue=true, group=MainframeEffect, parameterName=DisableShadows}
  )
  n54 = Ceil(None=NamedRerouteUsage())
  NamedRerouteDeclaration(None=Add(A=NamedRerouteUsage(), B=LinearInterpolate(A=@n53, Alpha=@n54) {constB=0})) {name=OutWPO, nodeColor=<rgba 1 0 0.1641 1>}
  n55 = GetMaterialAttributes(None=NamedRerouteDeclaration(None=FunctionInput() {bUsePreviewValueAsDefault=true, inputName=InMaterial, inputType=FunctionInput_MaterialAttributes}) {name=InMaterial, nodeColor=<rgba 0.3125 0 1 1>}) {attributeGetTypes [#1, #3]}
  NamedRerouteDeclaration(None=@n55.EmissiveColor) {name=InEmissive, nodeColor=<rgba 0.0078 1 0 1>}
  NamedRerouteDeclaration(None=@n55.WorldPositionOffset) {name=InWPO, nodeColor=<rgba 1 0 0.3281 1>}
  n56 = Multiply(A=TextureCoordinate()) {constB=5.5}
  n57 = TextureSample(UVs=@n56) {samplerType=SAMPLERTYPE_LinearGrayscale, texture=/Game/Global/Textures/Default/Noise/glitch_noise}
  NamedRerouteDeclaration(None=Power(Base=OneMinus(None=@n57.R)) {constExponent=16}) {name=Glitch, nodeColor=<rgba 0 0.1797 1 1>}
  n58 = PreSkinnedPosition()
  n59 = VectorParameter() {parameterName=SecondWipeOffset, defaultValue=<rgba 0 0 0 1>}
  n60 = ComponentMask(None=Multiply(A=Add(A=@n58, B=@n59.RGB), B=Constant() {r=-0.001})) {r=true}
  n61 = Multiply(A=NamedRerouteUsage(), B=ScalarParameter() {defaultValue=-0.1, parameterName=Param_1})
  n62 = NamedRerouteUsage()
  NamedRerouteDeclaration(None=SphereMask(
    A=VertexInterpolator(VS=@n60).PS
    B=Constant()
    Radius=Add(A=@n61, B=@n62)
    Hardness=ScalarParameter() {defaultValue=0.4, group=MainframeEffect, parameterName=SecondWipeRadius}
   )) {name=SecondWipe, nodeColor=<rgba 0.2891 0 1 1>}
  n63 = ScalarParameter() {defaultValue=0.3, parameterName=LineFadeTiling}
  n64 = VectorParameter() {parameterName=LineFadeOffset, defaultValue=<rgba 0.07 0.28 0 1>}
  n65 = TextureSample(UVs=Add(A=Multiply(A=@n36, B=@n63) {constB=0.25}, B=ComponentMask(None=@n64.RGB) {g=true, r=true})) {samplerType=SAMPLERTYPE_Masks, texture=&WindNoise_Mask_1}
  n66 = Add(A=@n65.R, B=@n39.R)
  NamedRerouteDeclaration(None=Saturate(None=Multiply(A=@n65.R, B=@n66))) {name=GaussianNoiseAnim, nodeColor=<rgba 1 0.7734 0 1>}
  n67 = Subtract(
   A=Time()
   B=Add(A=ScalarParameter() {parameterName=Timestamp}, B=ScalarParameter() {parameterName=TimestampDelay}) {constB=0, desc=Slight delay}
  )
  n68 = ScalarParameter() {defaultValue=2.5, parameterName=TimestampLength}
  NamedRerouteDeclaration(None=Saturate(None=Divide(A=@n67, B=@n68))) {name=OutTimestamp, nodeColor=<rgba 0 1 0.3438 1>}
  NamedRerouteUsage()
  Panner(Coordinate=@n56) {speedX=0.1}
  ScalarParameter() {defaultValue=300}
  ScalarParameter() {defaultValue=0.4248, group=MainframeEffect, parameterName=Time, sliderMax=1}
 ]}
